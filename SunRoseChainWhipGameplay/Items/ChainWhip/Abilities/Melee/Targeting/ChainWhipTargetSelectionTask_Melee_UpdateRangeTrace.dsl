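; Targeting selection task for the Chain Whip melee attack that supplies the trace used to determine the length of the chains.
; GetSourceLocation casts the request's source context to a FortPawn and returns its WeaponTowardsFocus targeting transform location.
; GetTraceDirection returns the forward vector of that pawn's base aim rotation.
; Both results must be valid on simulated and autonomous proxies, and both fall back to zero when the targeting subsystem is invalid.
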(fn GetSourceLocation
  (params (TargetingHandle :type Targeting%20Request%20Handle%20Structure%20%28by%20ref%29))
  (bind _returnvalue (Targeting|GetTargetingSubsystem TargetingHandle))
  ;[0.1416 0.1198 0.1667] GetSourceLocation: Source location for the targeting task which determines the length of the chains on the attack.
  ; It is important that this returned location is valid on simulated as well as autonomous proxies.
  (Utilities|FlowControl|Sequence
    (:then_0
      (Utilities|IsValid _returnvalue
        (:Is%20Valid
          (bind _asfort_pawn (Utilities|Casting|CastToFortPawn (Utilities|Struct|BreakTargetingSourceContext (TargetingSystem|TargetingTypes|GetTargetingSourceContext _returnvalue TargetingHandle))))
          (bind _returnvalue_location (Math|Transform|BreakTransform (Targeting|GetTargetingTransform _asfort_pawn "WeaponTowardsFocus" 0)))
          (return _returnvalue_location))))
    (:then_1
      (return "0, 0, 0"))))

(fn GetTraceDirection
  (params (TargetingHandle :type Targeting%20Request%20Handle%20Structure%20%28by%20ref%29))
  (bind _returnvalue (Targeting|GetTargetingSubsystem TargetingHandle))
  ;[0.02432 0.006016 0.02604] GetTraceDirection: The trace direction for the targeting task which determines the length of the chains on the attack.
  ; It is important that this returned direction is valid on simulated as well as autonomous proxies.
  (Utilities|FlowControl|Sequence
    (:then_0
      (Utilities|IsValid _returnvalue
        (:Is%20Valid
          (bind _asfort_pawn (Utilities|Casting|CastToFortPawn (Utilities|Struct|BreakTargetingSourceContext (TargetingSystem|TargetingTypes|GetTargetingSourceContext _returnvalue TargetingHandle))))
          (return (Math|Vector|GetForwardVector (Pawn|GetBaseAimRotation _asfort_pawn))))))
    (:then_1
      (return "0, 0, 0"))))
